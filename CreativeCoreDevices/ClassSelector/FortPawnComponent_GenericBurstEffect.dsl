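; Pawn component, filed under the Creative Class Selector device, that plays a one-shot burst VFX and sound on the owning character.
; ApplyEffect runs on authority, storing effect, sound, color and game time in a replicated struct.
; OnRep_Effect replays it only while start time plus EffectTimeAlive exceeds server world time, skipping stale bursts.
; Cosmetic-capable clients spawn the self-cleaning effects attached to the mesh's AttachSocket, passing Color to the VFX.
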(event Custom|Client%20Play%20Effect
  (params (Effect :type FXSystem%20Asset%20Object%20Reference) (Sound :type Sound%20Base%20Object%20Reference) (Color :type Linear%20Color%20Structure))
  ;[0.2086 0.3763 1 0.25] Clients - Applies the VFX and/or audio and attaches it to the owning character's mesh. This is purely a broadcast, so should only be used for burst effects that clean themselves u (no functionality built in here to stop existing effects).
  (Networking|CanExecuteCosmeticEvents
    (:True
      (bind _ascharacter (Utilities|Casting|CastToCharacter (Components|GetOwner)))
      (Utilities|IsValid Effect
        (:Is%20Valid
          (bind _returnvalue (Effects|SpawnFXSystemAttached :FXSystemAsset Effect :AttachToComponent (Variables|/Script/Engine.Character|GetMesh _ascharacter) :AttachPointName "AttachSocket" :Scale "1.000000,1.000000,1.000000" :LocationType "SnapToTarget" :bAutoDestroy true :PoolingMethod "AutoRelease" :bAutoActivate true))
          (Effects|Components|ParticleSystem|SetColorParameter _returnvalue "Color" Color))
        (:Is%20Not%20Valid
          _))
      (Utilities|IsValid Sound
        (:Is%20Valid
          (Audio|SpawnSoundAttached :Sound Sound :AttachToComponent (Variables|/Script/Engine.Character|GetMesh _ascharacter) :AttachPointName "AttachSocket" :LocationType "SnapToTarget" :bStopWhenAttachedToDestroyed false :VolumeMultiplier 1.0 :PitchMultiplier 1.0 :StartTime 0.0 :bAutoDestroy true))))))

(fn ApplyEffect
  (params (Effect :type FXSystem%20Asset%20Object%20Reference) (Sound :type Sound%20Base%20Object%20Reference) (Color :type Linear%20Color%20Structure))
  (Utilities|SwitchHasAuthority
    (:Authority
      (Variables|SetEffect_Effect_11_1A12445D4B2187AD31F5659CE01ECDB7 (Utilities|Struct|MakeInitialBurstEffectDetails Effect Sound Color (Utilities|Time|GetGameTimeinSeconds))))))

(fn CanApplyHolster
  (params (HolsterId :type Name%20%28by%20ref%29 :default "None"))
  (return false))

(fn OnRep_Effect
  (bind (_effect_effect_11_1a12445d4b2187ad31f5659ce01ecdb7 _effect_audio_10_859d03c04ffc9efb291c7492f73877c0 _effect_color_6_ec6717d1487d19a0d62eb49164c30c26 _effect_timestarted_14_1abbcc3247d968147111ccb736b681e7) (Utilities|Struct|BreakInitialBurstEffectDetails (Variables|GetEffect_Effect_11_1A12445D4B2187AD31F5659CE01ECDB7)))
  (if (> (+ _effect_timestarted_14_1abbcc3247d968147111ccb736b681e7 (Variables|Default|GetEffectTimeAlive)) (Fortnite|Sync|Time|GetServerWorldTimeSeconds))
    (CallFunction|ClientPlayEffect _effect_effect_11_1a12445d4b2187ad31f5659ce01ecdb7 _effect_audio_10_859d03c04ffc9efb291c7492f73877c0 _effect_color_6_ec6717d1487d19a0d62eb49164c30c26)))
